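; Child Blueprint of a SWAT van car prop that can start blinking lights when play begins.
; After the parent BeginPlay runs, a blinkylights flag gates activation of a particle system and a Black Monday light-blink sound played at that particle's location.
; Overlap and Tick events are empty, and the construction script only calls the parent.
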(event EventBeginPlay
  (|CallParentFunction)
  (if (Variables|Default|Getblinkylights)
    (Components|Activation|Activate (Variables|Default|GetParticleSystem) false)
    (Audio|PlaySoundatLocation "/Figment_S10/Athena/Sounds/Events/RiftBeacon/RiftBubble/AngryApple/BlackMonday_LightBlink_Cue.BlackMonday_LightBlink_Cue" (Transformation|GetWorldLocation (Variables|Default|GetParticleSystem)))))

(event Collision|EventActorBeginOverlap
  (params (OtherActor :type Actor%20Object%20Reference)))

(event EventTick
  (params (DeltaSeconds :type Float%20%28single-precision%29)))

(fn ConstructionScript
  (|CallParentFunction))
