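; Destructible gnome building prop that signals when it holds destruction loot.
; On BeginPlay it waits briefly, then attaches a particle system component raised above its origin only if the prop has destruction loot.
; When its death effects play, it fades out the gnome's ambient sound component.
; The construction script is empty.
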(event EventBeginPlay
  (Utilities|FlowControl|Delay 2)
  (if (Curve|HasDestructionLoot)
    (AddComponent|NODE_AddParticleSystemComponent-0 :RelativeTransform (Math|Transform|MakeTransform "0, 0,72.000000"))))

(event EventOnDeathPlayEffects
  (params (Damage :type Float%20%28single-precision%29) (DamageTags :type Gameplay%20Tag%20Container%20Structure%20%28by%20ref%29) (Momentum :type Vector :default "0, 0, 0") (HitInfo :type Hit%20Result%20Structure%20%28by%20ref%29) (InstigatedBy :type Fort%20Pawn%20Object%20Reference) (DamageCauser :type Actor%20Object%20Reference) (EffectContext :type Gameplay%20Effect%20Context%20Handle%20Structure))
  (Audio|Components|Audio|FadeOut (Variables|Default|GetGnome_Ambient_Sound) 0.5 0.0 "Linear" :!class /Script/Engine.AudioComponent))

(fn ConstructionScript)
