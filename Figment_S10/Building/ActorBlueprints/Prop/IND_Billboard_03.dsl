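; Billboard prop Blueprint with three configurable spotlights, each paired with a light ray mesh, plus a shared fog card.
; The cosmetic-only construction script maps lights to rays, then colors enabled pairs, sizing rays from attenuation radius and cone angles with random rotation.
; Disabled lights get zero radius and hidden rays.
; Custom primitive data drives per-light mesh emissive and fog color, with fog opacity scaled by active light count.
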(fn GetFogColorAndOpacity
  (bind (_fogcardcolorandopacity_r _fogcardcolorandopacity_g _fogcardcolorandopacity_b _fogcardcolorandopacity_a) (Math|Color|BreakColor (Variables|GetFogCardColorAndOpacity_R)))
  (return _fogcardcolorandopacity_r _fogcardcolorandopacity_g _fogcardcolorandopacity_b (* _fogcardcolorandopacity_a (Math|Float|Clamp(Float) (/ (Math|Conversions|ToFloat(Integer) (Variables|Default|GetNumberOfActiveLights)) 3.0)))))

(fn GetLightRayScale
  (params (InputPin :type Spot%20Light%20Component%20Object%20Reference))
  (return (Math|Vector|MakeVector (/ (Math|Float|Min(Float) (Variables|Default|GetMaxLightrayLength) (Variables|/Script/Engine.LocalLightComponent|GetAttenuationRadius InputPin)) 2e+02) (* (/ (+ (Variables|/Script/Engine.SpotLightComponent|GetInnerConeAngle InputPin) (Variables|/Script/Engine.SpotLightComponent|GetOuterConeAngle InputPin)) 4.0) (/ (Math|Float|Min(Float) (Variables|Default|GetMaxLightrayLength) (Variables|/Script/Engine.LocalLightComponent|GetAttenuationRadius InputPin)) 4e+03)) (* (/ (+ (Variables|/Script/Engine.SpotLightComponent|GetInnerConeAngle InputPin) (Variables|/Script/Engine.SpotLightComponent|GetOuterConeAngle InputPin)) 4.0) (/ (Math|Float|Min(Float) (Variables|Default|GetMaxLightrayLength) (Variables|/Script/Engine.LocalLightComponent|GetAttenuationRadius InputPin)) 4e+03)))))

(fn InitializeSpotlightMap
  (Utilities|Map|Clear (Variables|Default|GetLightsRays))
  (Utilities|Map|Add (Variables|Default|GetLightsRays) (Variables|Default|GetSpotLight01) (Variables|Default|GetLightRay01))
  (Utilities|Map|Add (Variables|Default|GetLightsRays) (Variables|Default|GetSpotLight02) (Variables|Default|GetLightRay02))
  (Utilities|Map|Add (Variables|Default|GetLightsRays) (Variables|Default|GetSpotLight03) (Variables|Default|GetLightRay03)))

(fn ConstructionScript
  (bind (_fogcardcolorandopacity_r _fogcardcolorandopacity_g _fogcardcolorandopacity_b _returnvalue) (CallFunction|GetFogColorandOpacity))
  (bind _lightsrays (Variables|Default|GetLightsRays))
  (bind (_meshemissivecolor_r _meshemissivecolor_g _meshemissivecolor_b _meshemissivecolor_a) (Math|Color|BreakColor (Variables|GetMeshEmissiveColor_R)))
  (bind (_lightcolor_r _lightcolor_g _lightcolor_b _lightcolor_a) (Math|Color|BreakColor (Variables|GetLightColor_R)))
  (Networking|CanExecuteCosmeticEvents
    (:True
      (Rendering|Material|SetMaterial (Variables|Debug|GetStaticMeshComponent) 1 "/Figment_S10/Environments/Sets/BlackMonday/Materials/M_BM_Billboard3Lights.M_BM_Billboard3Lights")
      (CallFunction|InitializeSpotlightMap)
      (bind _keys (Utilities|Map|Keys _lightsrays))
      (bind _values (Utilities|Map|Values _lightsrays))
      (for (_array_element _array_index) _keys
        (bind _output_get (Variables|Default|SetCurrentArrayIndex _array_index))
        (if (Utilities|Array|Get(aref) (Variables|Lights|GetSpotLightsEnabled) _array_index)
          (Transformation|SetRelativeRotation (Utilities|Array|Get(acopy) _values _output_get) (Math|Rotator|MakeRotator (Math|Random|RandomFloatinRange 0.0 3.6e+02)) false false)
          (Transformation|SetRelativeScale3D (Utilities|Array|Get(acopy) _values _output_get) (CallFunction|GetLightRayScale _array_element))
          (Rendering|Material|SetCustomPrimitiveDataVector4 (Utilities|Array|Get(acopy) _values _output_get) 28 (Math|Vector4|MakeVector4 _lightcolor_r _lightcolor_g _lightcolor_b _lightcolor_a))
          (Rendering|Components|Light|SetLightColor _array_element (Variables|Lights|GetLightColor) true)
          (Rendering|Material|SetCustomPrimitiveDataVector4 (Variables|Debug|GetStaticMeshComponent) 28 (Math|Vector4|MakeVector4 _meshemissivecolor_r _meshemissivecolor_g _meshemissivecolor_b _meshemissivecolor_a))
          (Rendering|Material|SetCustomPrimitiveDataFloat (Variables|Debug|GetStaticMeshComponent) (+ (Variables|Default|GetCurrentArrayIndex) 25) 1.0)
          (Variables|Default|SetHasLightEnabled true)
          (Math|Integer|IncrementInt (Variables|Default|GetNumberOfActiveLights))
          (else
            (Rendering|Lighting|SetAttenuationRadius _array_element 0.0)
            (Rendering|SetVisibility (Utilities|Array|Get(acopy) _values _output_get) false false)
            (Rendering|Material|SetCustomPrimitiveDataFloat (Variables|Debug|GetStaticMeshComponent) (+ (Variables|Default|GetCurrentArrayIndex) 25) 0.0))))
      (Rendering|Material|SetCustomPrimitiveDataVector4 (Variables|Default|GetFogCard) 28 (Math|Vector4|MakeVector4 _fogcardcolorandopacity_r _fogcardcolorandopacity_g _fogcardcolorandopacity_b _returnvalue))
      (Rendering|SetVisibility (Variables|Default|GetFogCard) (Variables|Default|GetHasLightEnabled) false))))
