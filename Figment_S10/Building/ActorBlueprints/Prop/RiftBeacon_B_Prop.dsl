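; Rift Beacon B building prop Blueprint that initializes the beacon's material state when play begins.
; On BeginPlay it sets the ArrayOn, RedAlpha, HoseAlpha and VibrationOn scalar parameters on the static mesh component's materials to zero, so those effects start switched off.
; The construction script is empty, and nothing in this Blueprint raises those values again.
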(event EventBeginPlay
  (bind _staticmeshcomponent (Variables|Debug|GetStaticMeshComponent))
  (Rendering|Material|SetScalarParameterValueonMaterials _staticmeshcomponent "ArrayOn" 0.0)
  (Rendering|Material|SetScalarParameterValueonMaterials _staticmeshcomponent "RedAlpha" 0.0)
  (Rendering|Material|SetScalarParameterValueonMaterials _staticmeshcomponent "HoseAlpha" 0.0)
  (Rendering|Material|SetScalarParameterValueonMaterials _staticmeshcomponent "VibrationOn" 0.0))

(fn ConstructionScript)
